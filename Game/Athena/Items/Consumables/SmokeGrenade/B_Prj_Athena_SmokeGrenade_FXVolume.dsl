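; Smoke grenade FX volume actor that shows a timed smoke cloud, destroying itself where cosmetic events cannot run.
; Construction script sizes the mesh and creates dynamic materials from the smoke radius.
; After a delay, one rate-scaled timeline fades materials and audio in, then reverses before lifespan ends.
; Outer Niagara smoke and autofire shroud detectors activate after fade-in and deactivate at fade-out, while a looping timer shows sound indicators.
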(event Custom|ShowAudioIndicator
  ;[1 1 1] Begin Play
  (Fortnite|ShowSoundIndicatorLocal :TrackedActor self :Location (Transformation|GetActorLocation self) :MaxAudibleDistance (Variables|Default|GetAV_Distance) :IndicatorType "Generic" :Tint "(R=0.000000,G=0.000000,B=0.000000,A=0.000000)" :OverrideIcon (Variables|Default|GetAV_Icon) :AttenuationMaxRangeMultiplier 1.0 :AttenuationIntensityMultiplier 1.0))

(event EventBeginPlay
  ;[1 1 1] Begin Play
  (Networking|CanExecuteCosmeticEvents
    (:True
      (Variables|Default|SetSmokebombLifespan (Math|Float|float+float (Variables|Default|GetSmokebombLifespan) (Variables|Default|GetSmokeDelayBeforeFadeIn) (Variables|Default|GetSmokeFadeOutTime)))
      (Actor|SetLifeSpan (Variables|Default|GetSmokebombLifespan))
      (Utilities|Time|SetTimerbyEvent (EventDispatchers|CreateEvent :!event Custom|FadeOut) (Utilities|ClampMinFloat (- (Variables|Default|GetSmokebombLifespan) (Variables|Default|GetSmokeFadeOutTime)) (Variables|Default|GetSmokeFadeInTime)))
      (Utilities|FlowControl|Delay (Variables|Default|GetSmokeDelayBeforeFadeIn))
      (CallFunction|FadeIn)
      (Development|SetHiddeninGame (Variables|Default|GetSmokeVolumeMesh) false false)
      (Utilities|Time|SetTimerbyEvent (EventDispatchers|CreateEvent :!event Custom|ShowAudioIndicator) 0.75 true true))
    (:False
      (Actor|DestroyActor))))

(event Custom|FadeIn
  ;[1 1 1] Timeline
  (Components|Timeline|SetPlayRate (Variables|B_Prj_Athena_SmokeGrenade_FXVolume|GetFadeSmokeTL) (Utilities|Reciprocal (Variables|Default|GetSmokeFadeInTime)))
  (bind (_direction _float) (Timeline|FadeSmokeTL PlayFromStart)
    (:Update
      (bind _output_get (Variables|Default|SetSmokeFadeValue _float))
      (Rendering|Material|SetScalarParameterValue (Variables|Default|GetSmokeVolumeMeshMID) "Fade" _output_get :!class /Script/Engine.MaterialInstanceDynamic)
      (Rendering|Material|SetScalarParameterValue (Variables|Default|GetSmokeScreen_MID) "Opacity" _output_get :!class /Script/Engine.MaterialInstanceDynamic)
      (Audio|Components|Audio|SetVolumeMultiplier (Variables|Default|GetAudio_ReleaseSmoke) (Variables|Default|GetSmokeFadeValue) :!class /Script/Engine.AudioComponent))
    (:Finished
      (switch Utilities|FlowControl|Switch|SwitchonETimelineDirection _direction
        (:Forward
          (Components|Activation|Activate (Variables|Default|GetNS_Athena_SmokeGrenade_OuterSmoke_01) false)
          (Class|AutofireShroudComponent|ActivateShroudDetectors (Variables|Default|GetAutofireSmokeShroudComponent) (Variables|Default|GetSmokebombRadius)))))))

(event Custom|FadeOut
  ;[1 1 1] Timeline
  (Components|Timeline|SetPlayRate (Variables|B_Prj_Athena_SmokeGrenade_FXVolume|GetFadeSmokeTL) (Utilities|Reciprocal (Variables|Default|GetSmokeFadeOutTime)))
  (Components|Activation|Deactivate (Variables|Default|GetNS_Athena_SmokeGrenade_OuterSmoke_01))
  (Default|DeactivateShroudDetectors (Variables|Default|GetAutofireSmokeShroudComponent))
  (Timeline|FadeSmokeTL ReverseFromEnd))

(fn ConstructionScript
  (Networking|CanExecuteCosmeticEvents
    (:True
      (Transformation|SetWorldScale3D (Variables|Default|GetSmokeVolumeMesh) (Math|Conversions|ToVector(Float) (Variables|Default|GetSmokebombRadius)))
      (bind _returnvalue (Rendering|Material|CreateDynamicMaterialInstance (Variables|Default|GetSmokeVolumeMesh) 0 "/Game/Athena/Items/Consumables/SmokeGrenade/M_Athena_SmokeGrenade_Outer_01.M_Athena_SmokeGrenade_Outer_01" "None" :!class /Script/Engine.PrimitiveComponent))
      (bind _output_get (Variables|Default|SetSmokeVolumeMeshMID _returnvalue))
      (Rendering|Material|SetScalarParameterValue _output_get "Fade" 1 :!class /Script/Engine.MaterialInstanceDynamic)
      (Rendering|Material|SetScalarParameterValue _output_get "Radius" (Variables|Default|GetSmokebombRadius) :!class /Script/Engine.MaterialInstanceDynamic)
      (bind _returnvalue_2 (Rendering|Material|CreateDynamicMaterialInstance (Variables|Default|GetSmokeScreen_Sphere) 0 "/Game/Athena/Items/Consumables/SmokeGrenade/M_SmokeGrenadeSphere.M_SmokeGrenadeSphere" "None" :!class /Script/Engine.PrimitiveComponent))
      (bind _output_get_2 (Variables|Default|SetSmokeScreen_MID _returnvalue_2))
      (Rendering|Material|SetScalarParameterValue _output_get_2 "Opacity" 0.0 :!class /Script/Engine.MaterialInstanceDynamic)
      (Rendering|Material|SetScalarParameterValue _output_get_2 "Radius" (* (Variables|Default|GetSmokebombRadius) 1.63) :!class /Script/Engine.MaterialInstanceDynamic))))
