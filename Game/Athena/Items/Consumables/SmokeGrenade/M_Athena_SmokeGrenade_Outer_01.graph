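# Translucent, two-sided lit material for the outer smoke layer of the Athena smoke grenade, tinted by particle color.
# Two panning samples of a tiling noise map and its normal map drive color, faint emissive and surface normals.
# The normals also drive world position offset, masked to the middle UV band and distance-faded from actor position.
# Opacity combines particle alpha, normal-perturbed Fresnel, depth fade against scene geometry and near-camera fade.
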Material /Game/Athena/Items/Consumables/SmokeGrenade/M_Athena_SmokeGrenade_Outer_01
graph {
 properties {
  bEnableResponsiveAA=true
  blendMode=BLEND_Translucent
  pixelDepthOffsetMode=PDOM_Legacy
  translucencyLightingMode=TLM_VolumetricPerVertexDirectional
  twoSided=true
  usedShadingModels=Default Lit
 }
 nodes [
  n1 = VectorParameter() {parameterName=BaseColorLow, defaultValue=<rgba 0.3 0.3 0.3 0.3>}
  n2 = ScalarParameter() {defaultValue=1, parameterName=U1}
  n3 = ScalarParameter() {defaultValue=1.3, parameterName=V1}
  n4 = Panner(Coordinate=Multiply(A=TextureCoordinate(), B=AppendVector(A=@n2, B=@n3))) {bFractionalPart=true, speedY=0.02}
  n5 = ScalarParameter() {defaultValue=2, parameterName=U2}
  n6 = ScalarParameter() {defaultValue=1.125, parameterName=V2}
  n7 = Panner(Coordinate=Multiply(A=TextureCoordinate(), B=AppendVector(A=@n5, B=@n6))) {bFractionalPart=true, speedY=0.0672}
  n8 = Divide(A=Add(
    A=TextureSample(UVs=@n4) {texture=&2ndInfectedMapLinear}.RGB
    B=TextureSample(UVs=@n7) {texture=&2ndInfectedMapLinear}.RGB
   ))
  n9 = ParticleColor()
  n10 = Multiply(
   A=Multiply(A=LinearInterpolate(A=@n1.RGB, Alpha=@n8) {constA=0.5}, B=@n9.RGB) {constA=1}
   B=VectorParameter() {parameterName=Color Tint, defaultValue=<rgba 1 1 1 1>}.RGB
  )
  MaterialFunctionCall(Emissive=Multiply(A=@n10, B=ScalarParameter() {defaultValue=0.02, parameterName=Emissive Fill Scale}) {constB=0.15}) {materialFunction=&MF_FX_EmissiveAdjustments}
  n11 = Normalize(VectorInput=Add(
    A=TextureSample(UVs=@n4) {samplerType=SAMPLERTYPE_Normal, texture=/Game/Global/Textures/Default/Tileable_Patterns/2ndInfectedMapNormal}.RGB
    B=TextureSample(UVs=@n7) {samplerType=SAMPLERTYPE_Normal, texture=&2ndInfectedMapNormal}.RGB
   ))
  n12 = VectorParameter() {parameterName=WPO Scale, defaultValue=<rgba 80 80 0 0>}
  n13 = MaterialFunctionCall(
   Position=ActorPositionWS()
   Fade_Start_Distance=ScalarParameter() {defaultValue=2000, parameterName=WPO Fade Start}
   Fade_End_Distance=ScalarParameter() {defaultValue=900, parameterName=WPO Fade End}
  ) {materialFunction=&MF_Distance_SyncedFade_FX}
  n14 = ComponentMask(None=TextureCoordinate()) {g=true}
  n15 = Constant() {r=0.5}
  n16 = ScalarParameter() {defaultValue=0.45, parameterName=WPO Mask Radius}
  n17 = ScalarParameter() {defaultValue=0.7, parameterName=WPO Mask Hardness}
  Multiply(
   A=Multiply(A=@n11, B=VertexNormalWS())
   B=Multiply(A=Multiply(A=@n12.RGB, B=@n13.Result), B=SphereMask(A=@n14, B=@n15, Radius=@n16, Hardness=@n17))
  )
  n18 = Transform(None=Multiply(A=@n11, B=VectorParameter() {parameterName=NormalIntensity, defaultValue=<rgba 0.3 0.3 1.5 0>}.RGB))
  n19 = VertexNormalWS()
  n20 = MaterialFunctionCall(Exponent=ScalarParameter() {defaultValue=3, parameterName=2ndFresnelLerp}) {materialFunction=&MF_Fresnel_FX}
  n21 = Multiply(
   A=MaterialFunctionCall(
    Brightness=ScalarParameter() {defaultValue=1.5, parameterName=FresnelBrightness}
    Exponent=ScalarParameter() {defaultValue=2.5, parameterName=FresnelExponent}
    Vertex / Pixel Normal=LinearInterpolate(A=@n18, B=@n19, Alpha=@n20.Fresnel)
   ) {materialFunction=&MF_Fresnel_FX}.Fresnel
   B=MaterialFunctionCall!(
    Distance Bias=ScalarParameter() {parameterName=FogBias}
    Distance Divisor=ScalarParameter() {defaultValue=75, parameterName=FogDivisor}
    Scene Depth
    Pixel Depth
   ) {materialFunction=/Game/Effects/Fort_Effects/Materials/_Functions/MF_DistanceBehindTranslucency_01}.Normalized
  )
  Multiply(
   A=MaterialFunctionCall(
    Opacity In=Multiply(A=@n9.A, B=@n21)
    Position=WorldPosition().XYZ
    Fade_Start_Distance=ScalarParameter() {defaultValue=40, parameterName=CameraFadeDistanceStart}
    Fade_End_Distance=ScalarParameter() {defaultValue=15, parameterName=CameraFadeDistanceEnd}
   ) {materialFunction=&MF_Distance_SyncedFade_FX}.Result
   B=ScalarParameter() {defaultValue=1, parameterName=Fade}
  )
  Reroute(None=@n10)
 ]
}
