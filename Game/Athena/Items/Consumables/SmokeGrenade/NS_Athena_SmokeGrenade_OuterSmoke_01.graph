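# Niagara system for the outer smoke cloud of the Athena smoke grenade, built from a single Smoke sprite emitter.
# Large, translucent, pale lavender sprites spawn steadily near the surface of a sphere sized just inside User.Radius and drift slowly upward.
# Each grows over a lifetime derived from System.Life Max, fading in and out while SubUV animation plays from a random start frame.
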NiagaraSystem /Game/Athena/Items/Consumables/SmokeGrenade/NS_Athena_SmokeGrenade_OuterSmoke_01
dependencies {
 usedDataInterfaces [
  &NiagaraDataInterfaceCurve
  &NiagaraDataInterfaceVectorCurve
  &NiagaraDataInterfaceSpriteRendererInfo
 ]
 usedDynamicInputs []
 usedModules [
  &self:SystemSpawnScript.SystemScriptSource.SystemScriptGraph.NiagaraNodeAssignment_0.SetVariables_#1
  &SystemState
 ]
}
emitters {Smoke {
  emitterUpdateScript [
   &EmitterState()
   &SpawnRate(SpawnRate=10, Use Spawn Probability=false)
  ]
  particleSpawnScript [
   &InitializeParticle(
    Color=<rgba 0.6997 0.6855 0.8438 0.3>
    Lifetime=&RandomRangeFloat(
     Minimum=&Subtract_Float(A=$System.Life Max, B=0.5)
     Maximum=$System.Life Max
     Evaluation Type=::NewEnumerator1
    )
    Sprite Size=&Vector2DFromFloat(Value=&RandomRangeFloat(Minimum=1200, Maximum=1500, Evaluation Type=::NewEnumerator1))
    Sprite Size Mode=::NewEnumerator4
    UsePositionOffset=false
    Write Material Random=true
   )
   &AddVelocity(Velocity=&RandomRangeVector(Minimum=<0 0 5>, Maximum=<0 0 20>, Evaluation Type=::NewEnumerator1))
   &ShapeLocation(
    Hemisphere Distribution=<1 0.5>
    MaskSpawnGroup=false
    Override Seed=false
    Sphere Radius=&Subtract_Float(A=$User.Radius, B=50)
    Sphere Surface Distribution=0.8
    UseOwnerScale=true
   )
  ]
  particleUpdateScript [
   &ParticleState()
   &ScaleSpriteSize(
    Scale Sprite Size Mode=::NewEnumerator0
    Uniform Scale Factor=&FloatFromCurve(FloatCurve=DataInterface() {propertyValues {Curve {keys [<at 0 0.6>,<at 1 1>]}}})
   )
   &SolveForcesAndVelocity(Clamp Velocity=false, Limit Acceleration=false)
   &ScaleColor(
    Scale Alpha=&FloatFromCurve(FloatCurve=DataInterface() {propertyValues {Curve {keys [
         {arriveTangent=4.533, interpMode=RCIM_Cubic, time=0, value=0}
         {arriveTangent=1, interpMode=RCIM_Cubic, leaveTangent=1, time=0.3, value=1}
         {arriveTangent=-1.4286, interpMode=RCIM_Cubic, leaveTangent=-1.4286, time=0.8, value=0.8}
         {arriveTangent=-0.1633, leaveTangent=-2.1558, tangentMode=RCTM_User, time=1, value=0}
        ]}}})
    Scale RGB=&VectorFromCurve(VectorCurve=DataInterface() {propertyValues {XCurve {keys [<at 0 1>,<at 1 1>]}, YCurve {keys [<at 0 1>,<at 1 1>]}, ZCurve {keys [<at 0 1>,<at 1 1>]}}})
    ScaleA=true
    ScaleRGB=false
   )
   &SubUVAnimation(
    Random Start Frame=true
    "Sprite Renderer "=DataInterface() {propertyValues {SpriteRenderer=&self:Smoke_0.NiagaraSpriteRendererProperties_0}}
    SubUV Lookup Index Scale=0.5
    UseEndFrame=false
    UseStartFrame=false
   )
  ]
  rendererClasses [&NiagaraSpriteRendererProperties]
  renderers [{rendererClass=&NiagaraSpriteRendererProperties}]
 }}
summary {
 systemName=NS_Athena_SmokeGrenade_OuterSmoke_01
 emitters [{emitterName=Smoke, rendererClasses [&NiagaraSpriteRendererProperties]}]
 userVariables [
  User.Duration {type=NiagaraFloat, defaultValue=9.5, description=""}
  User.Radius {type=NiagaraFloat, defaultValue=700, description=""}
 ]
}
